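# Niagara effect for the Fortanium rock that projects a directional beam along local X, scaled by User.DirLength, with User.Magnitude driving ring and ray intensity.
# Mesh rings travel outward over a soft glow mesh dimmed by time of day, with electric streak sprites spawning more often as DirLength rises.
# Ray sprites sampled from the rock mesh are gated by a DirLength threshold, in duplicate GPU and CPU emitters.
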NiagaraSystem /Fortanium/FX/NS_Fortanium_Rock_Direction
dependencies {
 usedDataInterfaces [
  &NiagaraDataInterfaceCurve
  &NiagaraDataInterfaceVector2DCurve
  &NiagaraDataInterfaceColorCurve
  &NiagaraDataInterfaceStaticMesh
  &NiagaraDataInterfaceSpriteRendererInfo
 ]
 usedDynamicInputs []
 usedModules [
  &SystemState
  &self:SystemSpawnScript.SystemScriptSource.SystemScriptGraph.NiagaraNodeAssignment_0.SetVariables_#1
 ]
}
emitters {
 Beam_Rings {
  emitterUpdateScript [
   &EmitterState()
   &SpawnRate(SpawnRate=&Lerp_Float(A=1, B=8, Alpha=$User.DirLength), Use Spawn Probability=false) {module=SpawnRate001}
  ]
  particleSpawnScript [
   /SetVariables(Particles.ColorA:LinearColor=<rgba 0 0.4618 2 1>) {isSetParameters=true}
   &InitializeParticle(
    Color Mode=::NewEnumerator0
    Lifetime=&Lerp_Float(A=2, B=0.7, Alpha=$User.DirLength)
    Mesh Scale=<3 3 3.2>
    Mesh Scale Mode=::NewEnumerator4
    UsePositionOffset=false
    Write Material Random=true
   )
   &InitialMeshOrientation(
    Mesh Axis To Orient=<0 0 1>
    Mesh Orientation Mode=::NewEnumerator2
    Rotation=&RandomRangeVector(Minimum=<0 0 0>, Maximum=<0 0 1>)
    Use Rotation Vector=true
   )
  ]
  particleUpdateScript [
   &ParticleState()
   /SetVariables(Particles.CameraOffset=20) {isSetParameters=true}
   /SetVariables(Particles.Position=&AddVectorToPosition(
     Position=<33 0 0>
     Vector=&MakeVector(
      X=&FloatFromCurve(
       FloatCurve=DataInterface() {propertyValues {Curve {keys [
           {arriveTangent=1.4387, interpMode=RCIM_Cubic, leaveTangent=1.4387, tangentMode=RCTM_User, time=0, value=0}
           {arriveTangent=0.3784, interpMode=RCIM_Cubic, leaveTangent=0.3784, tangentMode=RCTM_User, time=1, value=1}
          ]}}}
       Scale Curve=&Multiply_Float(
        A=&Lerp_Float(A=350, B=1000, Alpha=$User.DirLength)
        B=&FloatFromCurve(
         FloatCurve=DataInterface() {propertyValues {Curve {keys [
             {arriveTangent=9.0725, interpMode=RCIM_Cubic, leaveTangent=9.0725, tangentMode=RCTM_User, time=0, value=1}
             {interpMode=RCIM_Cubic, leaveTangent=-6.0265, tangentMode=RCTM_Break, time=0.1408, value=1.5}
             {arriveTangent=-0.0405, interpMode=RCIM_Cubic, leaveTangent=-0.0405, tangentMode=RCTM_User, time=0.5, value=1}
            ]}}}
         CurveIndex=$System.Age
        )
       )
      )
      Y=0
      Z=&FloatFromCurve(
       FloatCurve=DataInterface() {propertyValues {Curve {keys [
           {arriveTangent=1.4387, time=0, value=0.1}
           {arriveTangent=0.3784, interpMode=RCIM_Cubic, leaveTangent=0.3784, tangentMode=RCTM_User, time=1, value=1}
          ]}}}
       Scale Curve=&Lerp_Float(A=25, B=100, Alpha=$User.DirLength)
      )
     )
    )) {isSetParameters=true}
   &ScaleMeshSize(Scale Factor=&VectorFromFloat(Value=&FloatFromCurve(
      FloatCurve=DataInterface() {propertyValues {Curve {keys [{interpMode=RCIM_Cubic, tangentMode=RCTM_User, time=0, value=0.8}, {interpMode=RCIM_Cubic, tangentMode=RCTM_User, time=1, value=0.3}]}}}
      Scale Curve=0.45
     ))) {module=ScaleMeshSize001}
   &DynamicMaterialParameters(
    Index 0 Param 1=&FloatFromCurve(FloatCurve=DataInterface() {propertyValues {Curve {keys [
         {arriveTangent=-2.9934, interpMode=RCIM_Cubic, leaveTangent=-2.9934, tangentMode=RCTM_User, time=0, value=1}
         {interpMode=RCIM_Cubic, leaveTangent=-1.2287, time=1, value=-0.25}
        ]}}})
    Index 0 Param 2=&FloatFromCurve(
     FloatCurve=DataInterface() {propertyValues {Curve {keys [{tangentMode=RCTM_User, time=0, value=1}, {tangentMode=RCTM_User, time=1, value=0}]}}}
     Scale Curve=&RandomRangeFloat(Minimum=0, Maximum=0.5)
    )
    Index 0 Param 3=&FloatFromCurve(FloatCurve=DataInterface() {propertyValues {Curve {keys [
         {arriveTangent=2.2611, interpMode=RCIM_Cubic, leaveTangent=2.2611, tangentMode=RCTM_User, time=0, value=0}
         {arriveTangent=0.402, interpMode=RCIM_Cubic, leaveTangent=0.402, tangentMode=RCTM_User, time=1, value=1}
        ]}}})
    Index 0 Param 4=&FloatFromCurve(FloatCurve=DataInterface() {propertyValues {Curve {keys [{arriveTangent=-2.6143, interpMode=RCIM_Cubic, time=0, value=1}, <at 0.7 0>]}}})
    Module.Param0WriteEnabled=true
    Module.Param1WriteEnabled=true
    Module.Param2WriteEnabled=true
    Module.Param3WriteEnabled=true
   ) {module=DynamicMaterialParameters001}
   &Color(Color=&LerpLinearColors(
     StartColor=&MakeLinearColorFromVectorAndFloat(
      "Vector (RGB)"=&MakeVectorFromLinearColorRGB(Color=&MultiplyLinearColor(
        Linear Color=$Particles.ColorA
        Scale Factor=&MakeVector4_FromFloat(Float=&Lerp_Float(A=0.5, B=2, Alpha=$User.Magnitude))
       ))
      "Float (Alpha)"=&Multiply_Float(
       A=&Lerp_Float(A=0.25, B=1, Alpha=$User.Magnitude)
       B=&FloatFromCurve(FloatCurve=DataInterface() {propertyValues {Curve {keys [{interpMode=RCIM_Cubic, tangentMode=RCTM_User, time=0, value=0}, {interpMode=RCIM_Cubic, tangentMode=RCTM_User, time=1, value=1}]}}})
      )
     )
     EndColor=&MakeLinearColorFromVectorAndFloat(
      "Vector (RGB)"=&MakeVectorFromLinearColorRGB(Color=&MultiplyLinearColor(
        Linear Color=$Particles.ColorB
        Scale Factor=&MakeVector4_FromFloat(Float=&Lerp_Float(A=0.5, B=2, Alpha=$User.Magnitude))
       ))
      "Float (Alpha)"=&Multiply_Float(
       A=&Lerp_Float(A=0.25, B=9, Alpha=$User.Magnitude)
       B=&FloatFromCurve(FloatCurve=DataInterface() {propertyValues {Curve {keys [{interpMode=RCIM_Cubic, tangentMode=RCTM_User, time=0, value=0}, {interpMode=RCIM_Cubic, tangentMode=RCTM_User, time=1, value=1}]}}})
      )
     )
     LerpFactor=&FloatFromCurve(
      FloatCurve=DataInterface() {propertyValues {Curve {keys [
          {interpMode=RCIM_Cubic, tangentMode=RCTM_User, time=0, value=0}
          {arriveTangent=1.6728, interpMode=RCIM_Cubic, leaveTangent=1.6728, tangentMode=RCTM_User, time=1, value=1}
         ]}}}
      CurveIndex=$User.DirLength
      Scale Curve=&Multiply_Float(
       A=1
       B=&FloatFromCurve(
        FloatCurve=DataInterface() {propertyValues {Curve {keys [
            {interpMode=RCIM_Cubic, tangentMode=RCTM_User, time=0, value=1}
            {interpMode=RCIM_Cubic, leaveTangent=-2.6325, tangentMode=RCTM_Break, time=0.3, value=1.5}
            {tangentMode=RCTM_User, time=1, value=1}
           ]}}}
        CurveIndex=$System.Age
       )
      )
     )
    )) {module=Color002}
  ]
  rendererClasses [&NiagaraMeshRendererProperties]
  renderers [{rendererClass=&NiagaraMeshRendererProperties, rendererIndex=0}]
 }
 Beam_SoftGlow {
  emitterUpdateScript [
   &EmitterState(Inactive Response=::NewEnumerator1, Life Cycle Mode=::NewEnumerator1, Loop Duration=1, UseLoopDelay=false)
   &SpawnBurst_Instantaneous(Use Spawn Probability=false, UseLoopCountLimit=false)
   /SetVariables(Emitter.Desaturation=0.75, Emitter.SpeedY=-5) {isSetParameters=true}
  ]
  particleSpawnScript [
   &InitializeParticle(
    Color=&LerpLinearColors(StartColor=<rgba 0.0581 0.1022 1 1>, EndColor=<rgba 1.0352 0.1206 1.5 1>, LerpFactor=$User.DirLength)
    Lifetime=2
    Mesh Scale=&Lerp_Vector(A=<1.25 1 1>, B=<4.5 1 1>, Alpha=$User.DirLength)
    Mesh Scale Mode=::NewEnumerator4
    Position Offset=&Lerp_Vector(A=<1 0 0>, B=<30 0 0>, Alpha=$User.DirLength)
    UsePositionOffset=true
    Write Material Random=true
   )
   &InitialMeshOrientation(Mesh Orientation Mode=::NewEnumerator2, Orientation Vector=<1 0 0.1>, Use Rotation Vector=true)
   &DynamicMaterialParameters(
    Index 0 Param 1=&Lerp_Float(A=200, B=275, Alpha=$User.DirLength)
    Index 0 Param 4=1
    Module.Param0WriteEnabled=true
    Module.Param1WriteEnabled=true
    Module.Param2WriteEnabled=true
    Module.Param3WriteEnabled=true
   )
  ]
  particleUpdateScript [
   &ParticleState()
   /SetVariables(Particles.CameraOffset=20) {isSetParameters=true}
   &ScaleColor(
    Scale Alpha=&FloatFromCurve(
     FloatCurve=DataInterface() {propertyValues {Curve {keys [
         {interpMode=RCIM_Cubic, tangentMode=RCTM_User, time=0, value=0}
         {interpMode=RCIM_Cubic, tangentMode=RCTM_User, time=0.5, value=1}
         {interpMode=RCIM_Cubic, tangentMode=RCTM_User, time=1, value=0}
        ]}}}
     Scale Curve=&FNTimeOfDayScalar(Night Scale=1, Morning Scale=0.5, Evening Scale=0.5)
    )
    Scale RGB=&VectorFromFloat(Value=&Lerp_Float(
      A=4
      B=2
      Alpha=&FloatFromCurve(
       FloatCurve=DataInterface() {propertyValues {Curve {keys [
           {interpMode=RCIM_Cubic, tangentMode=RCTM_User, time=0, value=0}
           {arriveTangent=0.6039, interpMode=RCIM_Cubic, leaveTangent=0.6039, tangentMode=RCTM_User, time=0.85, value=0.2}
           {arriveTangentWeight=0.2241, interpMode=RCIM_Cubic, tangentMode=RCTM_User, time=1, value=1}
          ]}}}
       CurveIndex=$User.DirLength
      )
     ))
    ScaleA=true
    ScaleRGB=true
   ) {module=ScaleColor001}
  ]
  rendererClasses [&NiagaraMeshRendererProperties]
  renderers [{rendererClass=&NiagaraMeshRendererProperties, rendererIndex=0}]
 }
 ElectricStreaks {
  emitterUpdateScript [
   &EmitterState()
   &SpawnRate(
    Spawn Probability=&FloatFromCurve(
     FloatCurve=DataInterface() {propertyValues {Curve {keys [
         {interpMode=RCIM_Cubic, tangentMode=RCTM_User, time=0, value=0}
         {arriveTangent=0.7652, interpMode=RCIM_Cubic, leaveTangent=0.7652, tangentMode=RCTM_User, time=0.8, value=0.25}
         {interpMode=RCIM_Cubic, tangentMode=RCTM_User, time=1, value=1}
        ]}}}
     CurveIndex=$User.DirLength
    )
    SpawnRate=10
    Use Spawn Probability=true
   )
  ]
  particleSpawnScript [
   /Niagara/Modules/Spawn/Initialization/InitializeParticle(
    Color=&RandomRangeLinearColor(Minimum=<rgba 1.9477 1.0986 4 1>, Maximum=<rgba 4.19 2.4994 6 1>)
    Lifetime=&UniformRangedFloat(Minimum=0.55, Maximum=0.65)
    Module.Write Color=true
    Module.Write Lifetime=true
    Module.Write Mass=true
    Module.Write Material Random=true
    Module.Write Position=false
    Module.Write Scale=false
    Module.Write SpriteRotation=false
    Module.Write SpriteSize=true
    Position=&SimulationPosition()
    Sprite Size=&MakeVector2D(
     X=&RandomRangeFloat(Minimum=70, Maximum=90)
     Y=&FloatFromCurve(
      FloatCurve=DataInterface() {propertyValues {Curve {keys [
          {interpMode=RCIM_Cubic, tangentMode=RCTM_User, time=0, value=0.4}
          {interpMode=RCIM_Cubic, tangentMode=RCTM_User, time=0.5, value=1}
          {tangentMode=RCTM_User, time=1, value=0.4}
         ]}}}
      CurveIndex=&ReturnNormalizedExecIndex()
      Scale Curve=&Lerp_Float(
       A=450
       B=1000
       Alpha=&FloatFromCurve(
        FloatCurve=DataInterface() {propertyValues {Curve {keys [
            {interpMode=RCIM_Cubic, tangentMode=RCTM_User, time=0, value=0}
            {arriveTangent=2.6676, interpMode=RCIM_Cubic, leaveTangent=2.6676, tangentMode=RCTM_User, time=1, value=1}
           ]}}}
        CurveIndex=&RandomRangeFloat(Minimum=0, Maximum=1)
       )
      )
     )
    )
   )
   &DynamicMaterialParameters(
    Index 0 Param 1=&RandomRangeFloat(Minimum=0, Maximum=1)
    Index 0 Param 2=&Lerp_Float(A=0.7, B=0.125, Alpha=$Particles.MaterialRandom)
    Module.Param0WriteEnabled=true
    Module.Param1WriteEnabled=true
    Module.Param2WriteEnabled=false
    Module.Param3WriteEnabled=false
   )
   &CameraOffset(Camera Offset Amount=25)
   &ShapeLocation(
    MaskSpawnGroup=false
    Offset=&Lerp_Vector(A=<30 0 0>, B=<100 0 0>, Alpha=$User.DirLength)
    Offset Mode=::NewEnumerator0
    Override Seed=false
    Sphere Radius=50
    UseOwnerScale=true
   )
   &SpriteFacingAndAlignment(
    Sprite Alignment=&RotateVectorByQuaternion(
     Vector To Rotate=&RandomRangeVector(Minimum=<0.8 -0.2 -0.2>, Maximum=<1 0.2 0.2>)
     Quaternion=&MakeQuatFromAxisAngle(
      Axis=<1 0 0>
      AngleInDegrees=&Add_Float(
       A=&Lerp_Float(A=-55, B=55, Alpha=&ReturnNormalizedExecIndex())
       B=&RandomRangeFloat(Minimum=-15, Maximum=15)
      )
      NormalizeAxis=true
     )
    )
    WriteAlignment=true
    WriteFacing=false
   )
  ]
  particleUpdateScript [
   &ParticleState()
   &SolveForcesAndVelocity(Clamp Velocity=false, Limit Acceleration=false)
   &ScaleSpriteSize(
    Non-Uniform Curve Sprite Scale=DataInterface() {propertyValues {
      XCurve {keys [
        {arriveTangent=0.8718, interpMode=RCIM_Cubic, leaveTangent=0.8718, tangentMode=RCTM_User, time=0, value=0.8}
        {arriveTangentWeight=0.1795, interpMode=RCIM_Cubic, leaveTangentWeight=0.1795, tangentMode=RCTM_User, time=0.5, value=1}
        {arriveTangent=-3.9093, interpMode=RCIM_Cubic, leaveTangent=-3.9093, tangentMode=RCTM_User, time=1, value=0.2}
       ]}
      YCurve {keys [
        {arriveTangent=2.2429, interpMode=RCIM_Cubic, leaveTangent=2.2429, tangentMode=RCTM_User, time=0, value=0.5}
        {arriveTangent=0.3913, arriveTangentWeight=0.1667, interpMode=RCIM_Cubic, leaveTangent=0.3913, tangentMode=RCTM_User, time=0.5, value=1}
        {arriveTangent=0.1824, interpMode=RCIM_Cubic, leaveTangent=0.1824, tangentMode=RCTM_User, time=1, value=1.12}
       ]}
     }}
    Scale Sprite Size Mode=::NewEnumerator3
   )
   &DynamicMaterialParameters(
    Index 0 Param 3=&FloatFromCurve(FloatCurve=DataInterface() {propertyValues {Curve {keys [<at 0.5 0 RCIM_Cubic>, {arriveTangent=3.8263, interpMode=RCIM_Cubic, leaveTangent=3.8263, tangentMode=RCTM_User, time=1, value=1}]}}})
    Index 0 Param 4=&Divide_Float(
     A=/Niagara/DynamicInputs/Math/CalculateParticleRadius(
      Radius Calculation Type:ENiagaraCollisionRadiusOptions=::NewEnumerator0
      Particle Radius:NiagaraFloat=1
      Mesh Dimensions:Vector3f=<1 1 1>
      Method for Calculating Particle Radius:ENiagaraCalculateRadiusOptions=::NewEnumerator2
     )
     B=200
    )
    Index 1 Param 2=$System.RippleMult
    Index 1 Param 3=$System.RipplePan
    Index 1 Param 4=800
    Module.Param0WriteEnabled=false
    Module.Param0WriteEnabled1=true
    Module.Param1WriteEnabled=false
    Module.Param1WriteEnabled1=true
    Module.Param2WriteEnabled=true
    Module.Param2WriteEnabled1=true
    Module.Param3WriteEnabled=true
    Module.Param3WriteEnabled1=true
    Write Parameter Index 1=true
   ) {module=DynamicMaterialParameters001}
   &ScaleColor(
    Scale Alpha=&FloatFromCurve(FloatCurve=DataInterface() {propertyValues {Curve {keys [<at 0.5 1 RCIM_Cubic>,<at 1 0>]}}})
    Scale RGB=&MakeVectorFromLinearColorRGB(Color=&ColorFromCurve(
      ColorCurve=DataInterface() {propertyValues {
        AlphaCurve {keys []}
        BlueCurve {keys [
          <at 0.0485 1.8072 RCIM_Cubic>
          {arriveTangent=-2.3921, interpMode=RCIM_Cubic, leaveTangent=-2.3921, time=0.2983, value=1.6918}
          <at 0.534 0.6458>
         ]}
        GreenCurve {keys [
          <at 0.0485 0.4164 RCIM_Cubic>
          {arriveTangent=0.1445, interpMode=RCIM_Cubic, leaveTangent=0.1445, time=0.2983, value=0.3165}
          <at 0.534 0.4866>
         ]}
        RedCurve {keys [<at 0.0485 4 RCIM_Cubic>, {arriveTangent=-7.6897, interpMode=RCIM_Cubic, leaveTangent=-7.6897, time=0.2983, value=2}, <at 0.534 0.2666>]}
       }}
      CurveIndex=$Emitter.Age
     ))
    ScaleA=true
    ScaleRGB=true
   )
   &ScaleColor(
    Scale RGB=&VectorFromFloat(Value=&RandomRangeFloat(Minimum=0.75, Maximum=1.5))
    ScaleA=true
    ScaleRGB=false
   ) {module=ScaleColor001}
  ]
  rendererClasses [&NiagaraSpriteRendererProperties]
  renderers [{rendererClass=&NiagaraSpriteRendererProperties}]
 }
 Rays_Burst {
  simTarget=GPUComputeSim
  emitterUpdateScript [
   &EmitterState(Life Cycle Mode=::NewEnumerator1, Loop Behavior=::NewEnumerator1, Loop Duration=2, UseLoopDelay=false)
   &SpawnRate(SpawnRate=22, Use Spawn Probability=false)
  ]
  particleSpawnScript [
   /Niagara/Modules/Spawn/Initialization/V2/InitializeParticle(
    Color Mode=::NewEnumerator0
    Lifetime=2
    Sprite Size Max=<66 300>
    Sprite Size Min=<44 350>
    Sprite Size Mode=::NewEnumerator2
    UsePositionOffset=false
    Write Material Random=true
   )
   /Niagara/Modules/Spawn/Location/StaticMeshLocation(
    MaskSpawnGroup=false
    OffsetAlongNormal=false
    OverrideRandomSeed=false
    SamplePosOffset=false
    Static Mesh=DataInterface() {propertyValues {
      DefaultMesh=/Fortanium/Meshes/SM_Fortanium_Rock_01
      InstanceIndex=-1
      LODIndex=0
      PreviewMesh=null
      SoftSourceActor=null
      SourceMode=Default
      bAllowSamplingFromStreamingLODs=false
      bCaptureTransformsPerFrame=true
      bUsePhysicsBodyVelocity=false
      FilteredSockets []
      LODIndexUserParameter {parameter {name=None, typeDefHandle {registeredTypeIndex=108}, varData []}}
      MeshParameterBinding {parameter {name=None, typeDefHandle {registeredTypeIndex=1}, varData []}}
      SectionFilter {allowedMaterialSlots []}
     }}
   )
   &SpriteFacingAndAlignment(
    Sprite Alignment=&NormalizeVector(
     Vector To Normalize=&AddVectorToPosition(Position=$Particles.StaticMeshLocation.SampledPosition, Vector=<0 0 0>)
     Safe Normalize=true
     Safe Fallback Threshold=0
     Safe Fallback Vector=<0 0 0.001>
    )
    WriteAlignment=true
    WriteFacing=false
   )
   &Color(Color=&MakeLinearColorFromVectorAndFloat(
     "Vector (RGB)"=&MakeVectorFromLinearColorRGB(Color=<rgba 3 0.5176 3 2>)
     "Float (Alpha)"=&MakeCustomFloatFromBool(
      Bool=&SetBoolByFloatComparison(A=$User.DirLength, B=0.8, Comparison Type=::NewEnumerator2)
      True Float=1
      False Float=0
     )
    )) {module=Color001}
  ]
  particleUpdateScript [
   &ParticleState()
   &SubUVAnimation(
    End Frame Range Override=3
    "Sprite Renderer "=DataInterface() {propertyValues {SpriteRenderer=&self:Rays_Burst_0.NiagaraSpriteRendererProperties_0}}
    UseEndFrame=true
    UseStartFrame=true
   )
   &ScaleColor(
    Scale Alpha=&FloatFromCurve(
     FloatCurve=DataInterface() {propertyValues {Curve {keys [{interpMode=RCIM_Cubic, tangentMode=RCTM_User, time=0, value=0}, {interpMode=RCIM_Cubic, tangentMode=RCTM_User, time=1, value=1}]}}}
     Scale Curve=&Multiply_Float(A=$User.Magnitude, B=10)
    )
    ScaleA=true
    ScaleRGB=false
   )
   &ScaleColor(
    Scale Alpha=&FloatFromCurve(FloatCurve=DataInterface() {propertyValues {Curve {keys [{interpMode=RCIM_Cubic, tangentMode=RCTM_User, time=0, value=1}, {interpMode=RCIM_Cubic, tangentMode=RCTM_User, time=1, value=0}]}}})
    ScaleA=true
    ScaleRGB=false
   ) {module=ScaleColor001}
  ]
  rendererClasses [&NiagaraSpriteRendererProperties]
  renderers [{rendererClass=&NiagaraSpriteRendererProperties}]
 }
 Rays_Burst_CPU {
  emitterUpdateScript [
   &EmitterState(Life Cycle Mode=::NewEnumerator1, Loop Behavior=::NewEnumerator1, Loop Duration=2, UseLoopDelay=false)
   &SpawnRate(SpawnRate=22, Use Spawn Probability=false)
  ]
  particleSpawnScript [
   &InitializeParticle(
    Color Mode=::NewEnumerator0
    Lifetime=2
    Sprite Size Max=<66 300>
    Sprite Size Min=<44 350>
    Sprite Size Mode=::NewEnumerator2
    UsePositionOffset=false
    Write Material Random=true
   )
   &StaticMeshLocation(
    MaskSpawnGroup=false
    OffsetAlongNormal=false
    OverrideRandomSeed=false
    SamplePosOffset=false
    Static Mesh=DataInterface() {propertyValues {
      DefaultMesh=&SM_Fortanium_Rock_01
      InstanceIndex=-1
      LODIndex=0
      PreviewMesh=null
      SoftSourceActor=null
      SourceMode=Default
      bAllowSamplingFromStreamingLODs=false
      bCaptureTransformsPerFrame=true
      bUsePhysicsBodyVelocity=false
      FilteredSockets []
      LODIndexUserParameter {parameter {name=None, typeDefHandle {registeredTypeIndex=108}, varData []}}
      MeshParameterBinding {parameter {name=None, typeDefHandle {registeredTypeIndex=1}, varData []}}
      SectionFilter {allowedMaterialSlots []}
     }}
   )
   &SpriteFacingAndAlignment(
    Sprite Alignment=&NormalizeVector(
     Vector To Normalize=&AddVectorToPosition(Position=$Particles.StaticMeshLocation.SampledPosition, Vector=<0 0 0>)
     Safe Normalize=true
     Safe Fallback Threshold=0
     Safe Fallback Vector=<0 0 0.001>
    )
    WriteAlignment=true
    WriteFacing=false
   )
   &Color(Color=&MakeLinearColorFromVectorAndFloat(
     "Vector (RGB)"=&MakeVectorFromLinearColorRGB(Color=<rgba 3 0.5176 3 2>)
     "Float (Alpha)"=&MakeCustomFloatFromBool(
      Bool=&SetBoolByFloatComparison(A=$User.DirLength, B=0.8, Comparison Type=::NewEnumerator2)
      True Float=1
      False Float=0
     )
    )) {module=Color001}
  ]
  particleUpdateScript [
   &ParticleState()
   &SubUVAnimation(
    End Frame Range Override=3
    "Sprite Renderer "=DataInterface() {propertyValues {SpriteRenderer=&self:Rays_Burst_CPU_8.NiagaraSpriteRendererProperties_0}}
    UseEndFrame=true
    UseStartFrame=true
   )
   &ScaleColor(
    Scale Alpha=&FloatFromCurve(
     FloatCurve=DataInterface() {propertyValues {Curve {keys [{interpMode=RCIM_Cubic, tangentMode=RCTM_User, time=0, value=0}, {interpMode=RCIM_Cubic, tangentMode=RCTM_User, time=1, value=1}]}}}
     Scale Curve=&Multiply_Float(A=$User.Magnitude, B=10)
    )
    ScaleA=true
    ScaleRGB=false
   )
   &ScaleColor(
    Scale Alpha=&FloatFromCurve(FloatCurve=DataInterface() {propertyValues {Curve {keys [{interpMode=RCIM_Cubic, tangentMode=RCTM_User, time=0, value=1}, {interpMode=RCIM_Cubic, tangentMode=RCTM_User, time=1, value=0}]}}})
    ScaleA=true
    ScaleRGB=false
   ) {module=ScaleColor001}
  ]
  rendererClasses [&NiagaraSpriteRendererProperties]
  renderers [{rendererClass=&NiagaraSpriteRendererProperties}]
 }
}
summary {
 systemName=NS_Fortanium_Rock_Direction
 emitters [
  {emitterName=Beam_Rings, rendererClasses [&NiagaraMeshRendererProperties]}
  {emitterName=ElectricStreaks, rendererClasses [&NiagaraSpriteRendererProperties]}
  {emitterName=Rays_Burst, simTarget=GPUComputeSim, rendererClasses [&NiagaraSpriteRendererProperties]}
  {emitterName=Beam_SoftGlow, rendererClasses [&NiagaraMeshRendererProperties]}
  {emitterName=Rays_Burst_CPU, rendererClasses [&NiagaraSpriteRendererProperties]}
 ]
 userVariables [
  User.DirLength {type=NiagaraFloat, defaultValue=0, description=""}
  User.Magnitude {type=NiagaraFloat, defaultValue=0, description=""}
 ]
}
