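; Fortanium physics boulder prop that never dies: damage charges it, then it launches as a physics projectile.
; Each hit adds impulse, sets direction and restarts the charge timeline, which ends in a server impulse with optional hostile-target arc aim assist, then cooldown.
; Replicated values drive material charge and blink; clients add arrow Niagara and audio.
; It also buffs riders, skips or slows on water, and preserves cooldown through cloning.
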(event EventEndPlay
  (params (EndPlayReason :type EEndPlayReason%20Enum :default "Destroyed"))
  (bind _returnvalue (BuildingActor|PhysicsSimulationSupport|GetPhysicsObjectComponent self))
  (bind (_numberofcachedplayer _onboardlocation _offboardlocation _sessionlength) (PhysicsObject|GetCachedRidingPlayersInfo _returnvalue true))
  (FortPhysicsAnalytics|FireTelemetryEventonPhysicsRidingActorLifeEnd self _numberofcachedplayer _onboardlocation _offboardlocation _sessionlength))

(event EventBeginPlay
  ;[1 1 1] Begin Play
  (Networking|SwitchHasAuthority
    (:Authority
      (Variables|Default|SetCurrentChargeValue 1.0)
      (bind _returnvalue (BuildingActor|PhysicsSimulationSupport|GetPhysicsObjectComponent self))
      (PhysicsObject|BindEventtoOnRidingPawnAdded _returnvalue (EventDispatchers|CreateEvent :!event Custom|HandleRidingPawnAdded)))
    (:Remote
      (bind _returnvalue_2 (Rendering|Material|CreateDynamicMaterialInstance (Variables|Debug|GetStaticMeshComponent) 0 0 "None" :!class /Script/Engine.PrimitiveComponent))
      (Variables|Default|SetFortanium_MatInst _returnvalue_2)
      (Niagara|SetNiagaraVariable(Float) (Variables|Default|GetDirectionalArrow_FX) (Variables|Default|GetFX_Scale) (Variables|Default|GetFX_ScaleValue)) ; Niagara Variable Set\nScale\nThis is a set and forget variable, setting so FX\ncan scale between different meshes.
    )))

(event Collision|EventHit
  (params (MyComp :type Primitive%20Component%20Object%20Reference) (Other :type Actor%20Object%20Reference) (OtherComp :type Primitive%20Component%20Object%20Reference) (bSelfMoved :type Boolean) (HitLocation :type Vector :default "0, 0, 0") (HitNormal :type Vector :default "0, 0, 0") (NormalImpulse :type Vector :default "0, 0, 0") (Hit :type Hit%20Result%20Structure%20%28by%20ref%29))
  ;[1 1 1] On Hit (not damaged, just run into)
  (|CallParentFunction MyComp Other OtherComp bSelfMoved HitLocation HitNormal NormalImpulse Hit)
  ;;[0 0.5507 1] Only play FX if NOT in water (water FX handled separately in OnEnterWaterEvent)
  (if (not (Water|IsinWater (Variables|Default|GetFortWaterInteraction)))
    (Utilities|Casting|CastToLandscape Other
      (:then
        (Utilities|FlowControl|DoOnce false
          (:Completed
            (GameplayCue|ExecuteGameplayCueOnActor(Burst) self (Variables|Default|GetGCN%20LandTag) (GameplayCue|MakeGameplayCueParametersfromHitResult Hit)))))
      (:CastFailed
        (if (> (Math|Vector|VectorLength (Transformation|GetVelocity)) 5e+01)
          (GameplayCue|ExecuteGameplayCueOnActor(Burst) self (Variables|Default|GetGCN%20ImpactTag) (GameplayCue|MakeGameplayCueParametersfromHitResult Hit)))))))

(event Custom|ResetLandingFX
  ;[1 1 1] On Hit (not damaged, just run into)
  (Utilities|FlowControl|DoOnce false
    (:Completed
      (GameplayCue|ExecuteGameplayCueOnActor(Burst) self (Variables|Default|GetGCN%20LandTag) (GameplayCue|MakeGameplayCueParametersfromHitResult (stale Collision|EventHit Hit))))))

(event EventOnDamageServer
  (params (Damage :type Float%20%28single-precision%29) (DamageTags :type Gameplay%20Tag%20Container%20Structure%20%28by%20ref%29) (Momentum :type Vector :default "0, 0, 0") (HitInfo :type Hit%20Result%20Structure%20%28by%20ref%29) (InstigatedBy :type Controller%20Object%20Reference) (DamageCauser :type Actor%20Object%20Reference) (EffectContext :type Gameplay%20Effect%20Context%20Handle%20Structure))
  ;[1 1 1] On Damage Server
  (CallFunction|UpdateQuestHitFortanium InstigatedBy)
  (CallFunction|DamageHandlingforFortanium InstigatedBy Damage HitInfo DamageCauser DamageTags)
  (Variables|Default|SetFortaniumIsLaunching true)
  (GameplayCue|ExecuteGameplayCueOnActor(Burst) self (Variables|Default|GetGCN%20HitTag) (GameplayCue|MakeGameplayCueParametersfromHitResult HitInfo)))

(event Custom|DamageHandlingForFortanium
  (params (LastDamageDealer :type Controller%20Object%20Reference) (Damage :type Float%20%28double-precision%29) (Hit :type Hit%20Result%20Structure%20%28by%20ref%29) (DamageCauser :type Actor%20Object%20Reference) (DamageTags :type Gameplay%20Tag%20Container%20Structure%20%28by%20ref%29))
  ;[1 1 1] Set up for damage handling
  (Variables|Default|SetLastDamageDealer LastDamageDealer)
  (if (Variables|Default|GetCanInitiateLaunch)
    (Variables|Default|SetImpulseMagnitude (CallFunction|CalculateImpulseMagnitudefromDamage Damage))
    (if (GameplayTags|HasAnyTags DamageTags (Variables|Default|GetMeleeOrTraceWeapons))
      (bind _asfort_player_controller (Utilities|Casting|CastToFortPlayerController (Variables|Default|GetLastDamageDealer))
        (:then
          (Variables|Default|SetServerTargetImpulseDirection (CallFunction|CalculateCameraPitchOffset (Variables|/Script/Engine.PlayerController|GetPlayerCameraManager _asfort_player_controller) (ScalableFloat|GetValueatLevel (Variables|Default|GetImpulseDirectionPitchAngleOffset)))))
        (:CastFailed
          (Variables|Default|SetServerTargetImpulseDirection (CallFunction|CalculateImpulseDirection DamageTags Hit (Transformation|GetActorLocation DamageCauser))))) ; Can only use camera from fort player
      (CallFunction|HandleDamageEvent (Variables|Default|GetServerTargetImpulseDirection))
      (else
        _))))

(event Custom|HandleDamageEvent
  (params (Impulse%20Direction :type Vector))
  (bind _directionalarrow_fx (Variables|Default|GetDirectionalArrow_FX))
  ;[1 1 1] Build Up Charge
  (Variables|Default|SetcurrentImpulsePercentage (Default|GetImpulsePercentage))
  (Utilities|FlowControl|Sequence
    (:then_0
      ;;[1 1 1] Shared
      (Variables|Default|SetImpulse%20Direction Impulse%20Direction)
      (Health|SetHealth (Health|GetMaxHealth)) ; Reset health to max
      (Variables|Default|SetEnableImpulseOnDamage false)
      (Variables|Default|SetCurrentChargeValue (Math|Float|Lerp 1.0 3.0 (Variables|Default|GetcurrentImpulsePercentage)))
      (bind (_direction _chargefx _blink _ringspeed) (Timeline|ChargedState PlayFromStart)
        (:Update
          (Variables|Default|SetBlinkValue _blink)
          (Audio|Parameter|SetFloatParameter (Variables|Default|GetChargeAudio) "ExtraCharge" _chargefx :!class /Script/AudioExtensions.AudioParameterControllerInterface)
          (Niagara|SetNiagaraVariable(Float) (Variables|Default|GetDirectionalArrow_FX) (Variables|Default|GetDirectionalArrowRingSpeed) _ringspeed) ; Niagara Variable Set\nArrow Ring Speed\nBased on how long it has been since last damage taken\nUse to indicate how far along in the process the launch is
        )
        (:Finished
          (Networking|SwitchHasAuthority
            (:Authority
              (CallFunction|LaunchFortanium))))))
    (:then_1
      ;;[1 1 1] Client Only
      (Networking|SwitchHasAuthority
        (:Remote
          (if (< (Variables|Default|GetcurrentImpulsePercentage) 1.0)
            (Rendering|Material|SetCustomPrimitiveDataFloat (Variables|Debug|GetStaticMeshComponent) 2 (Utilities|Time|GetGameTimeinSeconds)) ; Play hit in material
            (CallFunction|SetAudioChargeParameter)
            (else
              (Utilities|FlowControl|Sequence
                (:then_0
                  _)
                (:then_1
                  (Utilities|FlowControl|DoOnce false
                    (:Completed
                      (bind _returnvalue (Audio|SpawnSoundAttached (Variables|Audio|GetMaxChargeReached) (Variables|Default|GetChargeAudio)))
                      (Variables|Default|SetMaxChargeReached_Comp _returnvalue)))))))
          (Components|Activation|Activate _directionalarrow_fx true)
          (Transformation|SetWorldRotation _directionalarrow_fx (Math|Conversions|RotationFromXVector (Variables|Default|GetImpulse%20Direction)) false false)
          (Niagara|SetNiagaraVariable(Float) (Variables|Default|GetDirectionalArrow_FX) (Variables|Default|GetDirectionalArrow_Length) (Default|GetImpulsePercentage)) ; Niagara Variable Set\nArrow Intensity\nBased on the impulse percentage from min to max amounts (value will be 0 through 1)\nUse to indicate the length and possibly intensity of the directional arrow FX
          (CallFunction|ResetLandingFX)
          (bind _prelaunchbuildup_comp (Variables|Default|GetPreLaunchBuildUp_Comp)
            (:then
              (Audio|Components|Audio|Play _prelaunchbuildup_comp 0.0))
            (:else
              (bind _returnvalue_2 (Audio|SpawnSoundAttached (Variables|Audio|GetPreLaunchBuildUp) (Variables|Debug|GetStaticMeshComponent)))
              (Variables|Default|SetPreLaunchBuildUp_Comp _returnvalue_2))))))))

(event Custom|ResetMaxChargeWarning
  ;[1 1 1] Build Up Charge
  ;;[1 1 1] Client Only
  (Utilities|FlowControl|DoOnce false
    (:Completed
      (bind _returnvalue (Audio|SpawnSoundAttached (Variables|Audio|GetMaxChargeReached) (Variables|Default|GetChargeAudio)))
      (Variables|Default|SetMaxChargeReached_Comp _returnvalue))))

(event Custom|LaunchFortanium
  (bind _impulse_direction (Variables|Default|GetImpulse%20Direction))
  ;[1 1 1] Launch Fortanium
  (Utilities|FlowControl|Sequence
    (:then_0
      ;;[1 1 1] Shared
      (Variables|Default|SetCachedChargeValue (Variables|Default|GetCurrentChargeValue))
      (Variables|Default|SetFortaniumIsLaunching false)
      (Variables|Default|SetEnableImpulseOnDamage false)
      (CallFunction|GrantRidingPawnsRidingGE)
      (bind _returnvalue (Utilities|Time|SetTimerbyEvent (EventDispatchers|CreateEvent :!event Custom|ResetInitiateLaunch) (ScalableFloat|GetValueatLevel (Variables|Default|GetCooldownTime))))
      (Variables|Default|SetInitatieLaunchCooldown _returnvalue)
      (bind _inflight_comp (Variables|Default|GetInFlight_Comp)
        (:then
          (Audio|Components|Audio|Play _inflight_comp 0.0))
        (:else
          (bind _returnvalue_2 (Audio|SpawnSoundAttached :Sound (Variables|Audio|GetInFlight) :AttachToComponent (Variables|Debug|GetStaticMeshComponent) :AttachPointName "None" :LocationType "SnapToTarget" :bStopWhenAttachedToDestroyed true :VolumeMultiplier 1.0 :PitchMultiplier 1.0 :StartTime 0.0 :bAutoDestroy true))
          (Variables|Default|SetInFlight_Comp _returnvalue_2)))
      (bind _returnvalue_3 (Utilities|Time|SetTimerbyEvent (delegate Custom|InFlightCheck) 0.1 true false 0.1))
      (Variables|Default|SetInFlightCheck_Handle _returnvalue_3))
    (:then_1
      (Networking|SwitchHasAuthority
        (:Authority
          ;;[1 1 1] Server Only
          (Variables|Default|SetCanInitiateLaunch false)
          (GameplayCue|ExecuteGameplayCue self (Variables|Default|GetGCN%20LaunchTag))
          (CallFunction|UpdateQuestLaunchFortanium)
          (bind (_trajectory _foundtarget) (CallFunction|ConsiderAimAssist _impulse_direction))
          (FortPhysicsSimulationLibrary|AddPhysicsImpulse self (select (and _foundtarget (ScalableFloat|GetValueasBool (Variables|Default|GetAimAssistEnabled))) _trajectory (* _impulse_direction (Variables|Default|GetImpulseMagnitude))) true "None" true false (Variables|Default|GetLastDamageDealer) "(MaxSimSizeAllowed=Large,bRejectEmbeddedInLandscape=False)")
          (CallFunction|UnchargeFortanium))
        (:Remote
          ;;[1 1 1] Client
          (CallFunction|SpawnRadialForceActorforFX)
          (Audio|Components|Audio|FadeOut (Variables|Default|GetChargeAudio) (Variables|Audio|GetChargeFadeOutTime) 0.0 "Linear" :!class /Script/Engine.AudioComponent)
          (CallFunction|ResetFX)
          (Audio|Components|Audio|Stop (Variables|Default|GetMaxChargeReached_Comp))
          (CallFunction|ResetMaxChargeWarning)
          (SoundIndicator|ShowSoundIndicator (Variables|Default|GetSound%20Indicator) (Transformation|GetActorLocation) "/Fortanium/Sound/Fortanium_PhysicsBoulder_Launch_Cue.Fortanium_PhysicsBoulder_Launch_Cue" 1.0 0))))))

(event Custom|InFlightCheck
  ;[1 1 1] Launch Fortanium
  ;;[1 1 1] Shared
  (if (< (Math|Vector|VectorLength (Transformation|GetVelocity)) 1e+02)
    (bind _inflight_comp (Variables|Default|GetInFlight_Comp)
      (:then
        (Audio|Components|Audio|FadeOut _inflight_comp 0.5 0.0 "Linear" :!class /Script/Engine.AudioComponent)
        (Utilities|Time|ClearandInvalidateTimerbyHandle (Variables|Default|GetInFlightCheck_Handle)))
      (:else))))

(event Custom|UnchargeFortanium
  ;[1 1 1] Uncharge Fortanium (happens on launch)
  (bind (_direction _chargevalue _blink) (Timeline|Uncharge PlayFromStart)
    (:Update
      (Variables|Default|SetCurrentChargeValue (Math|Float|Lerp 0.0 (Variables|Default|GetCachedChargeValue) _chargevalue))
      (Variables|Default|SetBlinkValue _blink))))

(event Custom|ResetInitiateLaunch
  ;[1 1 1] Reset Launch
  (Variables|Default|SetCanInitiateLaunch true)
  (Variables|Default|SetImpulseMagnitude 0.0)
  (bind (_direction _chargevalue) (Timeline|ChargeUp PlayFromStart)
    (:Update
      (Variables|Default|SetCurrentChargeValue _chargevalue))))

(event Custom|HandleRidingPawnAdded
  (params (RidingPawn :type Fort%20Player%20Pawn%20Object%20Reference))
  ;[1 1 1] Handle Pawn Riding
  ;;[1 1 1] Cached to apply RidingGE when launched
  (Utilities|Array|AddUnique (Variables|Default|GetCachedRidingPawns) RidingPawn)
  (Creative|GameplayEffects|SimpleApplyGameplayEffectToActor (Variables|Default|GetFallDamageImmunityGE) RidingPawn))

(event OnEnterWaterEvent(FortWaterInteraction)
  (params (WaterBody :type Fort%20Water%20Body%20Actor%20Object%20Reference) (WaterInteractionComponent :type Fort%20Water%20Interaction%20Component%20Object%20Reference) (bIsFirstBody :type Boolean))
  (bind _returnvalue (Transformation|GetVelocity))
  ;[1 1 1] Water Handling
  (Networking|SwitchHasAuthority
    (:Authority
      ;;[1 1 1] Water Skipping / Submersion speed handling
      (if (> (Math|Vector|VectorLength _returnvalue) (ScalableFloat|GetValueatLevel (Variables|Default|GetWaterSkipMinSpeed)))
        (if (CallFunction|TestWaterSkipAngle _returnvalue)
          (FortPhysicsSimulationLibrary|AddPhysicsImpulse self (CallFunction|CalculateSkipVelocityImpulse _returnvalue) true "None" true false (Variables|Default|GetLastDamageDealer) "(MaxSimSizeAllowed=Large,bRejectEmbeddedInLandscape=False)")
          (else
            _)))
      (FortPhysicsSimulationLibrary|AddPhysicsImpulse self (CallFunction|CalculateSubmergeVelocityImpulse _returnvalue) true "None" true false (Variables|Default|GetLastDamageDealer) "(MaxSimSizeAllowed=Large,bRejectEmbeddedInLandscape=False)"))
    (:Remote
      (CallFunction|PlayerWaterEffects (Variables|Audio|GetSFX_EnterWater) WaterBody (Variables|Default|GetFX_EnterWater)))))

(event OnExitWaterEvent(FortWaterInteraction)
  (params (WaterBody :type Fort%20Water%20Body%20Actor%20Object%20Reference) (WaterInteractionComponent :type Fort%20Water%20Interaction%20Component%20Object%20Reference) (bIsLastBody :type Boolean))
  ;[1 1 1] Water Handling
  (Networking|SwitchHasAuthority
    (:Remote
      (CallFunction|PlayerWaterEffects (Variables|Audio|GetSFX_ExitWater) WaterBody (Variables|Default|GetFX_ExitWater)))))

(event Custom|PlayerWaterEffects
  (params (SFXWater :type Sound%20Base%20Object%20Reference) (WaterBody :type Fort%20Water%20Body%20Actor%20Object%20Reference) (FXWater :type Niagara%20System%20Object%20Reference))
  ;[1 1 1] Play Effects for Enter / Exit Water
  (Utilities|FlowControl|Sequence
    (:then_0
      (Utilities|IsValid SFXWater
        (:Is%20Valid
          (Audio|PlaySoundatLocation SFXWater (Transformation|GetActorLocation)))))
    (:then_1
      (Utilities|IsValid FXWater
        (:Is%20Valid
          (bind _outwatersurfaceinfo (WaterBody|GetWaterSurfaceInfoStruct WaterBody (Transformation|GetActorLocation) true false false))
          (bind _returnvalue (Niagara|SpawnSystematLocation FXWater (Utilities|Struct|BreakWaterSurfaceInfo _outwatersurfaceinfo) (Math|Rotator|MakeRotfromX (.WaterPlaneNormal (Utilities|Struct|BreakWaterSurfaceInfo _outwatersurfaceinfo))) "1.000000,1.000000,1.000000" true true "AutoRelease"))
          (Effects|Components|ParticleSystem|SetFloatParameter _returnvalue "WaterDepth" (.WaterDepth (Utilities|Struct|BreakWaterSurfaceInfo _outwatersurfaceinfo)))
          (Effects|Components|ParticleSystem|SetFloatParameter _returnvalue "WaterBodyIndex" (Math|Conversions|ToFloat(Integer) (.WaterBodyIdx (Utilities|Struct|BreakWaterSurfaceInfo _outwatersurfaceinfo))))
          (Niagara|SetNiagaraVariable(Vector3) _returnvalue "waterPlaneLocation" (Utilities|Struct|BreakWaterSurfaceInfo _outwatersurfaceinfo))
          (Niagara|SetNiagaraVariable(Vector3) _returnvalue "waterPlaneNormal" (.WaterPlaneNormal (Utilities|Struct|BreakWaterSurfaceInfo _outwatersurfaceinfo)))
          (Niagara|SetNiagaraVariable(Float) _returnvalue "EffectScale" 1.0)
          (Niagara|SetNiagaraVariable(Int32) _returnvalue "SplashSpawn" 1))))))

(event EventOnPreCloningSetupStarted
  ;[1 1 1] Handle cloning logic
  ;;[1 1 1] Before cloning, store time remaining on cooldown timer (if not active this will be 0.0)
  (Variables|Default|SetCooldownTimeRemaining (Utilities|Time|GetTimerRemainingTimebyHandle (Variables|Default|GetInitatieLaunchCooldown))))

(event Building|EventOnBuildingActorInitialized
  (params (InitializationReason :type EFortBuildingInitializationReason%20Enum :default "StaticallyPlaced") (BuildingPersistentState :type EFortBuildingPersistentState%20Enum :default "Default"))
  ;[1 1 1] Handle cloning logic
  ;;[1 1 1] When new actor is spawned, set cooldown timer with correct time remaining
  (switch Utilities|FlowControl|Switch|SwitchonEFortBuildingInitializationReason InitializationReason
    (:Replaced
      _)
    (:DynamicBuilderPlaced
      _))
  (Utilities|Time|SetTimerbyEvent (EventDispatchers|CreateEvent :!event Custom|ResetInitiateLaunch) (Variables|Default|GetCooldownTimeRemaining)))

(fn CalculateCameraPitchOffset
  (params (CameraManager :type Player%20Camera%20Manager%20Object%20Reference) (PitchOffset :type Float%20%28double-precision%29))
  (return (Math|Rotator|GetRotationXVector (Math|Rotator|MakeRotator (.roll (Camera|GetCameraRotation CameraManager)) (Math|Float|SelectFloat (+ (.pitch (Camera|GetCameraRotation CameraManager)) PitchOffset) (Math|Float|SelectFloat (+ (.pitch (Camera|GetCameraRotation CameraManager)) PitchOffset) 9e+01 (> (+ (.pitch (Camera|GetCameraRotation CameraManager)) PitchOffset) 2.7e+02)) (< (+ (.pitch (Camera|GetCameraRotation CameraManager)) PitchOffset) 9e+01)) (.yaw (Camera|GetCameraRotation CameraManager))))))

(fn CalculateImpulseDirection
  (params (DamageTagContainer :type Gameplay%20Tag%20Container%20Structure%20%28by%20ref%29) (Hit :type Hit%20Result%20Structure%20%28by%20ref%29) (ActorLocation :type Vector))
  (bind (_bblockinghit _binitialoverlap _time _distance _location _impactpoint) (Collision|BreakHitResult Hit))
  (bind (_bblockinghit_2 _binitialoverlap_2 _time_2 _distance_2 _location_2) (Collision|BreakHitResult Hit))
  (if (GameplayTags|HasAnyTags DamageTagContainer (Variables|Default|GetExplosiveDamage))
    (return (Math|Vector|Normalize (- (Physics|GetCenterOfMass (Variables|Debug|GetStaticMeshComponent) "None") _location_2)))
    (else
      (return (Math|Vector|Normalize (- _impactpoint ActorLocation))))))

(fn CalculateImpulseMagnitudeFromDamage
  (params (Damage :type Float%20%28double-precision%29))
  (return (Math|Float|Clamp(Float) (+ (* Damage (ScalableFloat|GetValueatLevel (Variables|Default|GetDamageDealtMultiplier))) (Variables|Default|GetImpulseMagnitude)) (ScalableFloat|GetValueatLevel (Variables|Default|GetMinimumPhysicsImpulseAmount)) (ScalableFloat|GetValueatLevel (Variables|Default|GetMaximumPhysicsImpulseAmount)))))

(fn CalculateSkipVelocityImpulse
  (params (InVec :type Vector))
  (return (Math|Vector|MakeVector (* (.x InVec) (Math|Random|RandomFloatinRange (ScalableFloat|GetValueatLevel (Variables|Default|GetWaterSkipSpeedDecreaseMin)) (ScalableFloat|GetValueatLevel (Variables|Default|GetWaterSkipSpeedDecreaseMax)))) (* (.y InVec) (Math|Random|RandomFloatinRange (ScalableFloat|GetValueatLevel (Variables|Default|GetWaterSkipSpeedDecreaseMin)) (ScalableFloat|GetValueatLevel (Variables|Default|GetWaterSkipSpeedDecreaseMax)))) (* (Math|Float|Absolute(Float) (.z InVec)) (ScalableFloat|GetValueatLevel (Variables|Default|GetWaterSkipSpeedZMultiplier))))))

(fn CalculateSubmergeVelocityImpulse
  (params (Velocity :type Vector))
  (return (Math|Vector|SelectVector (* Velocity (ScalableFloat|GetValueatLevel (Variables|Default|GetWaterSubmergeSpeedDecrease))) (* (- (Math|Vector|VectorLength Velocity) (ScalableFloat|GetValueatLevel (Variables|Default|GetWaterSubmergeSpeedMax))) (Math|Vector|Normalize (* Velocity (ScalableFloat|GetValueatLevel (Variables|Default|GetWaterSubmergeSpeedDecrease))))) (<= (- (Math|Vector|VectorLength Velocity) (Math|Vector|VectorLength (* Velocity (ScalableFloat|GetValueatLevel (Variables|Default|GetWaterSubmergeSpeedDecrease))))) (ScalableFloat|GetValueatLevel (Variables|Default|GetWaterSubmergeSpeedMax))))))

(fn ConsiderAimAssist
  (params (UnalteredTrajectory :type Vector))
  (bind _returnvalue (Transformation|GetActorLocation))
  (bind _unalteredtrajectory (Variables|GetUnalteredTrajectory))
  (bind _returnvalue_2 (Math|Vector|Normalize _unalteredtrajectory))
  (bind _self self)
  (bind _impulsemagnitude (Variables|Default|GetImpulseMagnitude))
  (Utilities|IsValid (Variables|Default|GetLastDamageDealer)
    (:Is%20Valid
      (bind _outactors (Collision|SphereOverlapActors (Transformation|GetActorLocation) (Variables|Default|GetImpulseMagnitude) (Utilities|Array|MakeArray "ObjectTypeQuery6" "ObjectTypeQuery3" "ObjectTypeQuery2" "ObjectTypeQuery4" "ObjectTypeQuery7")))
      (bind _array_element (Utilities|Array|ForEachLoopwithBreak _outactors)
        (:LoopBody
          (if (Utilities|Enum|Equal(Enum) (Team|GetTeamAffiliation _array_element (Pawn|GetControlledPawn (Variables|Default|GetLastDamageDealer))) "Hostile")
            (bind _asfort_pawn (Utilities|Casting|CastToFortPawn _array_element))
            (bind (_outhit _returnvalue_3) (Collision|LineTraceByChannel _returnvalue (+ (FortPawn|GetMovementPredictionPosition _asfort_pawn (+ (/ (Class|AbilityFunctions|GetDistanceBetweenActors _asfort_pawn _self) _impulsemagnitude) 0.1)) "0.000000,0.000000,50.000000"))) ; No aim assist if you can't see the target
            (if (not _returnvalue_3)
              (bind (_outlaunchvelocity _returnvalue_4) (Game|SuggestProjectileVelocityCustomArc _returnvalue (+ (FortPawn|GetMovementPredictionPosition _asfort_pawn (+ (/ (Class|AbilityFunctions|GetDistanceBetweenActors _asfort_pawn _self) _impulsemagnitude) 0.1)) "0.000000,0.000000,50.000000") 0.0 0.95))
              (if _returnvalue_4
                (if (and (> (Math|Vector|DotProduct (Math|Vector|Normalize _outlaunchvelocity) _returnvalue_2) (ScalableFloat|GetValueatLevel (Variables|Default|GetAimAssistThreshold))) (> (Math|Vector|DotProduct (Math|Vector|Normalize _outlaunchvelocity) _returnvalue_2) (LocalVariables|GetTargetDot)))
                  (LocalVariables|SetTargetVelocity _outlaunchvelocity)
                  (LocalVariables|SetTargetDot (Math|Vector|DotProduct (Math|Vector|Normalize _outlaunchvelocity) _returnvalue_2)))))))
        (:Completed
          (return (LocalVariables|GetTargetVelocity) (> (LocalVariables|GetTargetDot) 0)))))
    (:Is%20Not%20Valid
      (return "0, 0, 0" false))))

(fn GetDamageDealtPercentage
  (return (/ (- (Variables|Default|GetImpulseMagnitude) (ScalableFloat|GetValueatLevel (Variables|Default|GetMinimumPhysicsImpulseAmount))) (- (ScalableFloat|GetValueatLevel (Variables|Default|GetMaximumPhysicsImpulseAmount)) (ScalableFloat|GetValueatLevel (Variables|Default|GetMinimumPhysicsImpulseAmount))))))

(fn GetImpulsePercentage
  (return (/ (- (Variables|Default|GetImpulseMagnitude) (ScalableFloat|GetValueatLevel (Variables|Default|GetMinimumPhysicsImpulseAmount))) (- (ScalableFloat|GetValueatLevel (Variables|Default|GetMaximumPhysicsImpulseAmount)) (ScalableFloat|GetValueatLevel (Variables|Default|GetMinimumPhysicsImpulseAmount))))))

(fn GrantRidingPawnsRidingGE
  (Networking|SwitchHasAuthority
    (:Authority
      (for _array_element (Variables|Default|GetCachedRidingPawns)
        (bind _output_get (LocalVariables|SetLocalRidingPawn _array_element))
        (Utilities|IsValid _output_get
          (:Is%20Valid
            ;[1 1 1] Only apply RidingGE if the cached pawn is still riding
            (if (== (Pawn|Components|CharacterMovement|GetMovementBaseObject (Variables|/Script/Engine.Character|GetCharacterMovement (LocalVariables|GetLocalRidingPawn))) (Variables|Debug|GetStaticMeshComponent))
              (Creative|GameplayEffects|SimpleApplyGameplayEffectToActor (Variables|Default|GetRidingGE) (LocalVariables|GetLocalRidingPawn)))))))))

(fn IterateToMax
  (params (IntegerRef :type Integer%20%28by%20ref%29) (Maximum :type Integer))
  (Variables|SetBy-RefVar IntegerRef (Math|Integer|Min(Integer) (+ IntegerRef 1) Maximum))
  (return))

(fn OnRep_BlinkValue
  (Rendering|Material|SetCustomPrimitiveDataFloat (Variables|Debug|GetStaticMeshComponent) 1 (Variables|Default|GetBlinkValue)))

(fn OnRep_CurrentChargeValue
  (Rendering|Material|SetCustomPrimitiveDataFloat (Variables|Debug|GetStaticMeshComponent) 0 (Variables|Default|GetCurrentChargeValue)))

(fn OnRep_FortaniumIsLaunching
  (if (not (Variables|Default|GetFortaniumIsLaunching))
    (CallFunction|ResetFX)))

(fn OnRep_FX_ChargeValue)

(fn RemoveFallDamageImmunityFromRidingPawns
  (Networking|SwitchHasAuthority
    (:Authority
      (for _array_element (Variables|Default|GetCachedRidingPawns)
        (bind _output_get (LocalVariables|SetLocalRidingPawn _array_element))
        (Utilities|IsValid _output_get
          (:Is%20Valid
            (GameplayEffects|RemoveActiveEffectswithGrantedTags (Ability|GetAbilitySystemComponent (LocalVariables|GetLocalRidingPawn)) "(GameplayTags=((TagName=\"Status.FallDamageImmunity\"),(TagName=\"Gameplay.Status.ImmuneFallingDamage.TillLand\")))")))))))

(fn ResetFX
  (Variables|Default|SetCachedChargeValue (Variables|Default|GetCurrentChargeValue))
  (Niagara|SetNiagaraVariable(Float) (Variables|Default|GetDirectionalArrow_FX) (Variables|Default|GetDirectionalArrow_Length) 0.0) ; Niagara Variable Set\nArrow Ring Speed\nBased on how long it has been since last damage taken\nUse to indicate how far along in the process the launch is
  (Niagara|SetNiagaraVariable(Float) (Variables|Default|GetDirectionalArrow_FX) (Variables|Default|GetDirectionalArrowRingSpeed) 0.0) ; Niagara Variable Set\nArrow Ring Speed\nBased on how long it has been since last damage taken\nUse to indicate how far along in the process the launch is
  (Components|Activation|Deactivate (Variables|Default|GetDirectionalArrow_FX)))

(fn SetAudioChargeParameter
  (if (not (Audio|Components|Audio|IsPlaying (Variables|Default|GetChargeAudio)))
    (Audio|Components|Audio|Play (Variables|Default|GetChargeAudio) 0.0))
  (Audio|Parameter|SetFloatParameter (Variables|Default|GetChargeAudio) "ChargeAmount" (Default|GetImpulsePercentage) :!class /Script/AudioExtensions.AudioParameterControllerInterface))

(fn ShouldDie
  (params (Damage :type Float%20%28single-precision%29) (EventInstigator :type Controller%20Object%20Reference) (DamageCauser :type Actor%20Object%20Reference) (bIsTest :type Boolean) (EffectContext :type Gameplay%20Effect%20Context%20Handle%20Structure%20%28by%20ref%29))
  (return false))

(fn SpawnRadialForceActorForFX
  (Utilities|AllowJanus
    (:Allow
      (bind _returnvalue (Development|GetConsoleVariableIntValue "r.PhysicsField.EnableField"))
      (if (== _returnvalue 1)
        (Game|SpawnActorfromClass "/Game/Effects/Niagara/Blueprints/PhsyicsFields/FN_RadialForce.FN_RadialForce_C" (Math|Transform|MakeTransform (Transformation|GetActorLocation)) "Undefined" "MultiplyWithRoot" 0 0.2 0.75 (Math|Float|Clamp(Float) (Variables|Default|GetImpulseMagnitude) 1e+02 3e+03) 1.5 0.0 0)))))

(fn TestWaterSkipAngle
  (params (Velocity :type Vector))
  (return (<= (Math|Float|Absolute(Float) (Class|AbilityFunctions|GetAngleBetweenTwoVectors Velocity (Math|Vector|MakeVector (.x Velocity) (.y Velocity)))) (ScalableFloat|GetValueatLevel (Variables|Default|GetWaterSkipAngle)))))

(fn UpdateQuestHitFortanium
  (params (Object :type Object%20Reference))
  (bind _asfort_player_controller (Utilities|Casting|CastToFortPlayerController Object))
  (Quest|SendComplexCustomStatEvent (Quests|GetQuestManager _asfort_player_controller "Athena") 0 "(GameplayTags=)" (Variables|Default|GetHitQuestTag) 1))

(fn UpdateQuestLaunchFortanium
  (bind _lastdamagedealer (Variables|Default|GetLastDamageDealer))
  (Utilities|IsValid _lastdamagedealer
    (:Is%20Valid
      (Quest|SendComplexCustomStatEvent (Quests|GetQuestManager (Fortnite|GetFortPlayerControllerfromActor _lastdamagedealer) "Athena") 0 "(GameplayTags=)" (Variables|Default|GetLaunchQuestTag) 1))) ; Invalid when rocks are hit by non pawns (ie, other rocks)
)

(fn ConstructionScript
  (|CallParentFunction))
