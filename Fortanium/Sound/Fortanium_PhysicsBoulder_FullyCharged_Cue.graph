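# Sound cue for a Fortanium physics boulder reaching full charge.
# A random node picks one of two equally weighted wave variants, then a modulator slightly varies pitch downward at constant volume.
# Playback uses a dedicated full-charge rock attenuation asset and a third-person ranged weapon fire sound class, with a concurrency override capping simultaneous instances at a small count.
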SoundCue /Fortanium/Sound/Fortanium_PhysicsBoulder_FullyCharged_Cue
root=Modulator_2
nodes [
 {id=Modulator_2, children [Random_1], properties {pitchMax=1, pitchMin=0.9, volumeMax=1, volumeMin=1}}
 {id=Random_1, children [WavePlayer_0,WavePlayer_1], properties {weights [1,1]}}
 {id=WavePlayer_0, properties {soundWaveAssetPtr=&/Fortanium_PhysicsBoulder_FullyCharged_01}}
 {id=WavePlayer_1, properties {soundWaveAssetPtr=&/Fortanium_PhysicsBoulder_FullyCharged_02}}
]
properties {
 attenuationSettings=/Game/Sounds/Attenuation/Fort_FortaniumPhysicsRocksFullCharge_Att
 bOverrideConcurrency=true
 maxDistance=10000
 pitchMultiplier=1.02
 soundClassObject=&Fort_WeaponRanged_Fire_3P_PSM
 volumeMultiplier=1
 attenuationOverrides {falloffDistance=5000, attenuationShapeExtents {x=500}}
 concurrencyOverrides {maxCount=3}
}
