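# Sound cue for the hit impact of a Fortanium physics boulder, routed to the Fort_Impacts sound class.
# A Random node picks among six equally weighted impact waves and feeds a Modulator that holds pitch fixed.
# Cue-level attenuation overrides the referenced footsteps attenuation with logarithmic falloff and distance-based low-pass filtering.
# Playback is boosted in volume, primed on load and governed by the shared player melee weapon concurrency set.
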SoundCue /Fortanium/Sound/Fortanium_PhysicsBoulder_HitImpact_Cue
root=Modulator_0
nodes [
 {id=Modulator_0, children [Random_0], properties {pitchMax=1, pitchMin=1}}
 {
  id=Random_0
  children [WavePlayer_5, WavePlayer_6, WavePlayer_7, WavePlayer_8, WavePlayer_9, WavePlayer_10]
  properties {weights [1,1,1,1,1,1]}
 }
 {id=WavePlayer_10, properties {soundWaveAssetPtr=&/Fortanium_PhysicsBoulder_HitImpact_06}}
 {id=WavePlayer_5, properties {soundWaveAssetPtr=&/Fortanium_PhysicsBoulder_HitImpact_01}}
 {id=WavePlayer_6, properties {soundWaveAssetPtr=&/Fortanium_PhysicsBoulder_HitImpact_02}}
 {id=WavePlayer_7, properties {soundWaveAssetPtr=&/Fortanium_PhysicsBoulder_HitImpact_03}}
 {id=WavePlayer_8, properties {soundWaveAssetPtr=&/Fortanium_PhysicsBoulder_HitImpact_04}}
 {id=WavePlayer_9, properties {soundWaveAssetPtr=&/Fortanium_PhysicsBoulder_HitImpact_05}}
]
properties {
 attenuationSettings=/Game/Sounds/Attenuation/Fort_Player_Footsteps_Att
 bOverrideAttenuation=true
 bPrimeOnLoad=true
 maxDistance=1100
 soundClassObject=&Fort_Impacts
 volumeMultiplier=2.4
 attenuationOverrides {
  bAttenuateWithLPF=true
  distanceAlgorithm=Logarithmic
  falloffDistance=800
  lPFRadiusMax=800
  lPFRadiusMin=400
  attenuationShapeExtents {x=300}
 }
 concurrencyOverrides {bLimitToOwner=true, maxCount=2, resolutionRule=StopOldest}
 concurrencySet [&Player_Melee_Weapon_Conc]
}
