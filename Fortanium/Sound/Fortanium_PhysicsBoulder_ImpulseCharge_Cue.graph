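# Sound cue for the Fortanium physics boulder's impulse charge, mixing looping layers with spatial attenuation and limited concurrency.
# The ChargeAmount parameter drives volume and pitch of a pitched-up glow loop and a pitched-down spark build-up loop.
# ExtraCharge drives a louder-mixed pair of launch warning loops.
# Envelopes raise that pair's pitch over about two seconds and fade one warning loop in after a second.
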SoundCue /Fortanium/Sound/Fortanium_PhysicsBoulder_ImpulseCharge_Cue
root=Mixer_1
nodes [
 {
  id=Enveloper_0
  children [WavePlayer_1]
  properties {
   pitchCurve {editorCurveData {keys [{time=0}]}}
   volumeCurve {editorCurveData {keys [<at 1 0>, {interpMode=RCIM_Cubic, tangentMode=RCTM_Break, time=2, value=0.9739}]}}
  }
 }
 {
  id=Enveloper_3
  children [Modulator_3]
  properties {
   pitchCurve {editorCurveData {keys [
      {interpMode=RCIM_Cubic, tangentMode=RCTM_User, time=0.0136, value=0.8096}
      {arriveTangent=0.432, interpMode=RCIM_Cubic, leaveTangent=0.432, tangentMode=RCTM_User, time=2.1479, value=1.032}
     ]}}
   volumeCurve {editorCurveData {keys [<at 0 1>]}}
  }
 }
 {id=Mixer_0, children [Enveloper_0,WavePlayer_2], properties {inputVolume [1,1]}}
 {id=Mixer_1, children [ModulatorContinuous_0, ModulatorContinuous_3, ModulatorContinuous_1], properties {inputVolume [0.7,0.7,2.3]}}
 {
  id=ModulatorContinuous_0
  children [Modulator_2]
  properties {
   pitchModulationParams {default=0, maxOutput=1.1, minOutput=0.9, parameterName=ChargeAmount}
   volumeModulationParams {default=0, minOutput=0.2, parameterName=ChargeAmount}
  }
 }
 {
  id=ModulatorContinuous_1
  children [Enveloper_3]
  properties {
   pitchModulationParams {default=0, maxOutput=1.25, minOutput=1, parameterName=ExtraCharge}
   volumeModulationParams {default=0, parameterName=ExtraCharge}
  }
 }
 {
  id=ModulatorContinuous_3
  children [Modulator_0]
  properties {
   pitchModulationParams {default=0, maxOutput=1.1, minOutput=0.9, parameterName=ChargeAmount}
   volumeModulationParams {default=0, minOutput=0.2, parameterName=ChargeAmount}
  }
 }
 {id=Modulator_0, children [WavePlayer_0], properties {pitchMax=0.7, pitchMin=0.7, volumeMax=1, volumeMin=1}}
 {id=Modulator_2, children [WavePlayer_6], properties {pitchMax=1.5, pitchMin=1.5, volumeMax=1, volumeMin=1}}
 {id=Modulator_3, children [Mixer_0], properties {pitchMax=1, pitchMin=1, volumeMax=1, volumeMin=1}}
 {id=WavePlayer_0, properties {bLooping=true, soundWaveAssetPtr=&/Fortanium_PhysicsBoulder_SparkBuildUpLoop}}
 {id=WavePlayer_1, properties {bLooping=true, soundWaveAssetPtr=&/Fortanium_PhysicsBoulder_LaunchWarningLoop}}
 {id=WavePlayer_2, properties {bLooping=true, soundWaveAssetPtr=&/Fortanium_PhysicsBoulder_LaunchWarningLoop2}}
 {id=WavePlayer_6, properties {bLooping=true, soundWaveAssetPtr=&/Fortanium_PhysicsBoulder_ChargeGlow}}
]
properties {
 attenuationSettings=/Game/Sounds/Attenuation/Fort_FortaniumPhysicsRocks_Att
 bOverrideConcurrency=true
 maxDistance=10000
 pitchMultiplier=1.02
 soundClassObject=&Fort_WeaponRanged_Fire_3P_PSM
 volumeMultiplier=0.9
 attenuationOverrides {distanceAlgorithm=Logarithmic, falloffDistance=10000, nonSpatializedRadiusStart=100, stereoSpread=1, attenuationShapeExtents {x=500}}
 concurrencyOverrides {maxCount=3}
}
