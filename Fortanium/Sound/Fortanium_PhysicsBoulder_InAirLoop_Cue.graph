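# Sound cue for the looping in-air sound of a Fortanium physics boulder.
# A mixer layers two looping waves, a launch warning loop pitched up through a modulator and an in-air wind loop at normal pitch.
# The cue overrides attenuation and concurrency for long-range falloff with a small voice limit, and routes to a third-person ranged weapon fire sound class.
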SoundCue /Fortanium/Sound/Fortanium_PhysicsBoulder_InAirLoop_Cue
root=Mixer_1
nodes [
 {id=Mixer_1, children [Modulator_2,Modulator_1], properties {inputVolume [1.2,1.2]}}
 {id=Modulator_1, children [WavePlayer_9], properties {pitchMax=1, pitchMin=1, volumeMax=1, volumeMin=1}}
 {id=Modulator_2, children [WavePlayer_0], properties {pitchMax=1.2, pitchMin=1.2, volumeMax=1, volumeMin=1}}
 {id=WavePlayer_0, properties {bLooping=true, soundWaveAssetPtr=&/Fortanium_PhysicsBoulder_LaunchWarningLoop2}}
 {id=WavePlayer_9, properties {bLooping=true, soundWaveAssetPtr=&/Fortanium_PhysicsBoulder_InAirWind}}
]
properties {
 attenuationSettings=/Game/Sounds/Attenuation/Fort_FortaniumPhysicsRocksLaunch_Att
 bOverrideAttenuation=true
 bOverrideConcurrency=true
 maxDistance=12500
 pitchMultiplier=1.02
 soundClassObject=&Fort_WeaponRanged_Fire_3P_PSM
 volumeMultiplier=1.2
 attenuationOverrides {falloffDistance=12000, attenuationShapeExtents {x=500}}
 concurrencyOverrides {maxCount=3}
}
