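# Sound cue for launching a Fortanium physics boulder, layering a launch sound with a sparks sound.
# A mixer blends two randomized four-wave branches: the launch body, slightly pitch-varied and briefly delayed, and the sparks, pitched up and mixed louder.
# It uses a Fortanium rocks-launch attenuation asset, a third-person ranged weapon-fire sound class, and a concurrency override limiting simultaneous instances.
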SoundCue /Fortanium/Sound/Fortanium_PhysicsBoulder_Launch_Cue
root=Mixer_0
nodes [
 {id=Delay_0, children [Random_1], properties {delayMax=0.1, delayMin=0.1}}
 {id=Mixer_0, children [Modulator_2,Modulator_0], properties {inputVolume [1,2]}}
 {id=Modulator_0, children [Random_0], properties {pitchMax=1.5, pitchMin=1.5}}
 {id=Modulator_2, children [Delay_0], properties {pitchMin=0.98, volumeMax=1, volumeMin=1}}
 {id=Random_0, children [WavePlayer_1,WavePlayer_2,WavePlayer_3,WavePlayer_4], properties {weights [1,1,1,1]}}
 {id=Random_1, children [WavePlayer_5,WavePlayer_6,WavePlayer_7,WavePlayer_8], properties {weights [1,1,1,1]}}
 {id=WavePlayer_1, properties {soundWaveAssetPtr=&/Fortanium_PhysicsBoulder_LaunchSparks_01}}
 {id=WavePlayer_2, properties {soundWaveAssetPtr=&/Fortanium_PhysicsBoulder_LaunchSparks_02}}
 {id=WavePlayer_3, properties {soundWaveAssetPtr=&/Fortanium_PhysicsBoulder_LaunchSparks_03}}
 {id=WavePlayer_4, properties {soundWaveAssetPtr=&/Fortanium_PhysicsBoulder_LaunchSparks_04}}
 {id=WavePlayer_5, properties {soundWaveAssetPtr=&/Fortanium_PhysicsBoulder_Launch_01}}
 {id=WavePlayer_6, properties {soundWaveAssetPtr=&/Fortanium_PhysicsBoulder_Launch_02}}
 {id=WavePlayer_7, properties {soundWaveAssetPtr=&/Fortanium_PhysicsBoulder_Launch_03}}
 {id=WavePlayer_8, properties {soundWaveAssetPtr=&/Fortanium_PhysicsBoulder_Launch_04}}
]
properties {
 attenuationSettings=/Game/Sounds/Attenuation/Fort_FortaniumPhysicsRocksLaunch_Att
 bOverrideConcurrency=true
 maxDistance=10000
 pitchMultiplier=1.02
 soundClassObject=&Fort_WeaponRanged_Fire_3P_PSM
 volumeMultiplier=0.5
 attenuationOverrides {falloffDistance=10000, attenuationShapeExtents {x=500}}
 concurrencyOverrides {maxCount=3}
}
