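# Sound cue for the Fortanium physics boulder water splash, playing one of several splash variations.
# An equally weighted Random node picks one of five waves and feeds a Modulator that applies a fixed pitch raise with no random spread.
# The cue overrides attenuation and concurrency inline, despite also referencing a rock-launch attenuation asset, limits simultaneous instances, and routes to a third-person ranged weapon fire sound class.
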SoundCue /Fortanium/Sound/Fortanium_PhysicsBoulder_WaterSplash_Cue
root=Modulator_2
nodes [
 {id=Modulator_2, children [Random_0], properties {pitchMax=1.2, pitchMin=1.2, volumeMax=1, volumeMin=1}}
 {id=Random_0, children [WavePlayer_1,WavePlayer_2,WavePlayer_3,WavePlayer_4,WavePlayer_5], properties {weights [1,1,1,1,1]}}
 {id=WavePlayer_1, properties {soundWaveAssetPtr=&/Fortanium_PhysicsBoulder_WaterSplash_01}}
 {id=WavePlayer_2, properties {soundWaveAssetPtr=&/Fortanium_PhysicsBoulder_WaterSplash_02}}
 {id=WavePlayer_3, properties {soundWaveAssetPtr=&/Fortanium_PhysicsBoulder_WaterSplash_03}}
 {id=WavePlayer_4, properties {soundWaveAssetPtr=&/Fortanium_PhysicsBoulder_WaterSplash_04}}
 {id=WavePlayer_5, properties {soundWaveAssetPtr=&/Fortanium_PhysicsBoulder_WaterSplash_05}}
]
properties {
 attenuationSettings=/Game/Sounds/Attenuation/Fort_FortaniumPhysicsRocksLaunch_Att
 bOverrideAttenuation=true
 bOverrideConcurrency=true
 maxDistance=10500
 pitchMultiplier=1.02
 soundClassObject=&Fort_WeaponRanged_Fire_3P_PSM
 volumeMultiplier=1
 attenuationOverrides {falloffDistance=10000, attenuationShapeExtents {x=500}}
 concurrencyOverrides {maxCount=3}
}
